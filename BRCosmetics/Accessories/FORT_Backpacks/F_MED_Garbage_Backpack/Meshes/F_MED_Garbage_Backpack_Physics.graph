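# Physics asset for the F_MED_Garbage_Backpack cosmetic, defining bodies and joints for its dynamic ear bones.
# A capsule on backpack_root anchors box bodies for a single-segment left ear and a two-segment right ear chain.
# Angular limits keep twist tight on every joint, while the left ear swings mostly on swing2 and both right ear joints mostly on swing1.
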PhysicsAsset /BRCosmetics/Accessories/FORT_Backpacks/F_MED_Garbage_Backpack/Meshes/F_MED_Garbage_Backpack_Physics
bodies {
 backpack_root [{shapeType=Capsule, length=18.0386, radius=3.4236, center=<-2.0275 -8.2416 -0.839>, rotation=<rot 0 -85.3567 -175.722>}]
 dyn_ear_01_r [{shapeType=Box, extentX=3.4292, extentY=0.75, extentZ=7.0333, center=<1.5016 -0.0372 -0.0043>}]
 dyn_ear_02_r [{shapeType=Box, extentX=1.7779, extentY=7.0355, extentZ=7.8251, center=<3.0264 0.2604 -0.122>, rotation=<rot 0 90 0>}]
 dyn_ear_l [{shapeType=Box, extentX=8.2682, extentY=1.6522, extentZ=9.7972, center=<3.6417 0.0446 0.0905>, rotation=<rot 0 0 90>}]
}
constraints [
 {bone1Name=dyn_ear_l, bone2Name=backpack_root, swing1LimitDegrees=12, swing2LimitDegrees=50, twistLimitDegrees=3.4}
 {bone1Name=dyn_ear_01_r, bone2Name=backpack_root, swing1LimitDegrees=45, swing2LimitDegrees=5, twistLimitDegrees=3.4}
 {bone1Name=dyn_ear_02_r, bone2Name=dyn_ear_01_r, swing1LimitDegrees=45, swing2LimitDegrees=5, twistLimitDegrees=3.4}
]
