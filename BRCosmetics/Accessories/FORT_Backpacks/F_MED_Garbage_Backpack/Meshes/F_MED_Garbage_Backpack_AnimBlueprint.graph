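# Animation Blueprint for the Garbage backpack cosmetic, adding procedural secondary motion to its bones from the reference pose.
# Anim Dynamics swings wheel bones 01 and 03 under overridden gravity, and Bone Driven Controllers map their Y rotation, negated, onto X of wheels 02 and 04.
# A RigidBody node then simulates the backpack's physics asset relative to backpack_root.
# A similar nose-bone pair is not connected to the output, and the event graph has no logic.
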AnimBlueprint /BRCosmetics/Accessories/FORT_Backpacks/F_MED_Garbage_Backpack/Meshes/F_MED_Garbage_Backpack_AnimBlueprint
target_skeleton=&/F_MED_Garbage_Backpack_Skeleton
graphs [
 AnimGraph {nodes [
   n1 = Anim Dynamics - Bone: dyn_l_nose(Alpha=1, ExternalForce="0,0,0") {
    angularDampingOverride=0.98
    bGravityOverrideInSimSpace=true
    bOverrideAngularDamping=true
    bUseGravityOverride=true
    alphaBoolBlend {blendOption=Linear}
    boundBone {boneName=dyn_l_nose}
    componentAppliedLinearAccClamp=<3000 3000 3000>
    componentLinearAccScale=<0.3 0.3 0.3>
    componentLinearVelScale=<0.3 0.3 0.3>
    gravityOverride {z=-980}
    physicsBodyDefinitions [{
      sphereCollisionRadius=0
      boundBone {boneName=dyn_l_nose}
      boxExtents=<5 10 5>
      constraintSetup {twistAxis=AxisY, angularLimitsMax=<5 0 0>, angularLimitsMin=<-20 0 0>, angularTarget=<0 0 0>}
      localJointOffset=<0 -10 0>
     }]
   }
   "dyn_r_nose.rotateX = dyn_l_nose.rotateY - Bone Driven Controller"!(ComponentPose:Component Space Pose Link Structure=@n1, Alpha:float=1) -> (Pose:Component Space Pose Link Structure) {
    is=title AnimGraphNode_BoneDrivenController
    bAffectTargetRotationX=true
    sourceComponent=RotationY
    alphaBoolBlend {blendOption=Linear}
    sourceBone {boneName=dyn_l_nose}
    targetBone {boneName=dyn_r_nose}
   }
   n2 = Anim Dynamics - Bone: dyn_wheel_01(ComponentPose=Local To Component(LocalPose=Local Space Ref Pose()), Alpha=1, ExternalForce="0,0,0") {
    angularDampingOverride=0.99
    bGravityOverrideInSimSpace=true
    bOverrideAngularDamping=true
    bUseGravityOverride=true
    alphaBoolBlend {blendOption=Linear}
    boundBone {boneName=dyn_wheel_01}
    componentAppliedLinearAccClamp=<3000 3000 3000>
    componentLinearAccScale=<0.3 0.3 0.3>
    componentLinearVelScale=<0.3 0.3 0.3>
    gravityOverride {z=-980}
    physicsBodyDefinitions [{
      sphereCollisionRadius=0
      boundBone {boneName=dyn_wheel_01}
      boxExtents=<3 5 5>
      constraintSetup {twistAxis=AxisY, angularLimitsMax=<90 0 0>, angularLimitsMin=<-90 0 0>, angularTarget=<0 0 0>}
      localJointOffset=<0 0.3 0>
     }]
   }
   n3 = "dyn_wheel_02.rotateX = dyn_wheel_01.rotateY * -1 - Bone Driven Controller"!(ComponentPose:Component Space Pose Link Structure=@n2, Alpha:float=1) -> (Pose:Component Space Pose Link Structure) {
    is=title AnimGraphNode_BoneDrivenController
    bAffectTargetRotationX=true
    multiplier=-1
    sourceComponent=RotationY
    alphaBoolBlend {blendOption=Linear}
    sourceBone {boneName=dyn_wheel_01}
    targetBone {boneName=dyn_wheel_02}
   }
   n4 = Anim Dynamics - Bone: dyn_wheel_03(ComponentPose=@n3, Alpha=1, ExternalForce="0,0,0") {
    angularDampingOverride=0.99
    bGravityOverrideInSimSpace=true
    bOverrideAngularDamping=true
    bUseGravityOverride=true
    alphaBoolBlend {blendOption=Linear}
    boundBone {boneName=dyn_wheel_03}
    componentAppliedLinearAccClamp=<3000 3000 3000>
    componentLinearAccScale=<0.3 0.3 0.3>
    componentLinearVelScale=<0.3 0.3 0.3>
    gravityOverride {z=-980}
    physicsBodyDefinitions [{
      sphereCollisionRadius=0
      boundBone {boneName=dyn_wheel_03}
      boxExtents=<3 5 5>
      constraintSetup {twistAxis=AxisY, angularLimitsMax=<90 0 0>, angularLimitsMin=<-90 0 0>, angularTarget=<0 0 0>}
      localJointOffset=<0 0.3 0>
     }]
   }
   n5 = "dyn_wheel_04.rotateX = dyn_wheel_03.rotateY * -1 - Bone Driven Controller"!(ComponentPose:Component Space Pose Link Structure=@n4, Alpha:float=1) -> (Pose:Component Space Pose Link Structure) {
    is=title AnimGraphNode_BoneDrivenController
    bAffectTargetRotationX=true
    multiplier=-1
    sourceComponent=RotationY
    alphaBoolBlend {blendOption=Linear}
    sourceBone {boneName=dyn_wheel_03}
    targetBone {boneName=dyn_wheel_04}
   }
   n6 = RigidBody(ComponentPose=@n5) {
    bOverrideWorldGravity=true
    overridePhysicsAsset=&/F_MED_Garbage_Backpack_Physics
    simulationSpace=BaseBoneSpace
    alphaBoolBlend {blendOption=Linear}
    baseBoneRef {boneName=backpack_root}
    componentAppliedLinearAccClamp=<1000 1000 1000>
    componentLinearAccScale=<0.8 0.8 0.8>
    componentLinearVelScale=<0.6 0.6 0.6>
    overrideWorldGravity {z=-980}
   }
   Output Pose(Result=Component To Local(ComponentPose=@n6)) {layerGroup=None}
  ]}
 EventGraph {nodes [Event BlueprintUpdateAnimation()]}
]
variables []
